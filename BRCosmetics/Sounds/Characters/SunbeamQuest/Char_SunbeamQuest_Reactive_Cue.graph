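# Reactive character foley sound cue for the SunbeamQuest cosmetic, playing one of three React waves.
# After a short fixed delay, a switch on the StatValue parameter routes all five numbered inputs to the same equally weighted random node, while its default input stays silent.
# Basic foley attenuation and the player foley class apply, and per-owner concurrency allows one instance, stopping the oldest.
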SoundCue /BRCosmetics/Sounds/Characters/SunbeamQuest/Char_SunbeamQuest_Reactive_Cue
root=Delay_0
nodes [
 {id=Delay_0, children [Switch_0], properties {delayMax=0.05, delayMin=0.05}}
 {id=Random_0, children [WavePlayer_0,WavePlayer_1,WavePlayer_2], properties {weights [1,1,1]}}
 {id=Switch_0, children [null,Random_0,Random_0,Random_0,Random_0,Random_0], properties {intParameterName=StatValue}}
 {id=WavePlayer_0, properties {soundWaveAssetPtr=&/Character_SunbeamQuest_React_01}}
 {id=WavePlayer_1, properties {soundWaveAssetPtr=&/Character_SunbeamQuest_React_02}}
 {id=WavePlayer_2, properties {soundWaveAssetPtr=&/Character_SunbeamQuest_React_03}}
]
properties {
 attenuationSettings=&Foley_Attenuation_Basic_Att
 bOverrideConcurrency=true
 maxDistance=1536
 pitchMultiplier=0.95
 soundClassObject=&Fort_Player_Foley
 volumeMultiplier=4
 concurrencyOverrides {bLimitToOwner=true, maxCount=1, resolutionRule=StopOldest}
}
